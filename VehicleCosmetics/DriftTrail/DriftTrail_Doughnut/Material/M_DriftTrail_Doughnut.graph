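# Translucent unlit Niagara material for the Doughnut drift trail vehicle cosmetic, enabled for ribbons and mesh particles.
# Two tiled, panning flame-mask samples on UV1 are summed and shaped by a V-edge falloff into layered, contrast-boosted masks.
# The "Trail?" static switch selects between a sine-power wheel mask and U-gradient shaping.
# Particle color is square-rooted, monochromatically shifted and blended through those masks, feeding a shared emissive adjustment function.
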Material /VehicleCosmetics/DriftTrail/DriftTrail_Doughnut/Material/M_DriftTrail_Doughnut
graph {
 properties {
  bUsedWithNiagaraMeshParticles=true
  bUsedWithNiagaraRibbons=true
  blendMode=BLEND_Translucent
  pixelDepthOffsetMode=PDOM_Legacy
  shadingModel=MSM_Unlit
  usedShadingModels=Unlit
 }
 nodes [
  n1 = AppendVector(
   A=ScalarParameter() {defaultValue=1, parameterName=Texture 1 Tiling X}
   B=ScalarParameter() {defaultValue=1, parameterName=Texture 1 Tiling V}
  )
  n2 = TextureCoordinate() {coordinateIndex=1}
  n3 = TextureSample(UVs=Panner(Coordinate=Multiply(A=@n1, B=@n2), Speed=VectorParameter() {parameterName=Speed_Panner_1, defaultValue=<rgba -1 0.15 0 0>}.RGB) {speedX=1.5}) {texture=/DealershipSharedLibrary/FX/Textures/TexturePackage/T_VCVFX_FlameMask_RGB}
  n4 = AppendVector(
   A=ScalarParameter() {defaultValue=1, parameterName=Texture 2 Tiling X}
   B=ScalarParameter() {defaultValue=1, parameterName=Texture 2 Tiling V}
  )
  n5 = TextureCoordinate() {coordinateIndex=1}
  n6 = TextureSample(UVs=Panner(Coordinate=Multiply(A=@n4, B=@n5), Speed=VectorParameter() {parameterName=Speed_Panner_2, defaultValue=<rgba -1.5 0.5 0 1>}.RGB) {speedX=2.5}) {texture=&T_VCVFX_FlameMask_RGB}
  AppendVector(A=@n3.G, B=@n6.R)
  n7 = TextureCoordinate()
  n8 = ComponentMask(None=@n7) {g=true}
  n9 = Multiply(A=@n8, B=OneMinus(None=@n8))
  n10 = Multiply(A=@n9, B=Constant() {r=5})
  n11 = ComponentMask(None=Multiply(A=Add(A=@n3.G, B=@n6.R), B=@n10)) {g=true}
  n12 = ComponentMask(None=@n7) {r=true}
  n13 = OneMinus(None=@n12)
  n14 = StaticSwitchParameter(True=@n11, False=Add(A=Multiply(A=@n13, B=@n9), B=@n11)) {parameterName=Trail?}
  n15 = TextureCoordinate()
  n16 = Power(Base=Sine(None=ComponentMask(None=@n15) {g=true}) {period=2}, Exp=ScalarParameter() {defaultValue=2.8, parameterName=Mask_Wheel}) {constExponent=0.98}
  n17 = StaticSwitchParameter(True=@n16, False=Multiply(A=Constant() {r=2}, B=@n13)) {parameterName=Trail?}
  MaterialExpressionConstantClamp_2 = Saturate(None=ConstantBiasScale(None=Multiply(A=@n14, B=@n17)) {bias=-0.25, scale=2})
  n18 = Multiply(A=Constant() {r=8}, B=Multiply(A=@n12, B=@n10))
  MaterialExpressionConstantClamp_0 = Saturate(None=Multiply(A=@MaterialExpressionConstantClamp_2, B=StaticSwitchParameter(True=@n16, False=@n18) {parameterName=Trail?}))
  MaterialExpressionVertexColor_0 = ParticleColor()
  n19 = SquareRoot(None=Abs(None=@MaterialExpressionVertexColor_0.RGB))
  n20 = Constant() {r=1.15}
  n21 = LinearInterpolate(
   A=Multiply(
    A=Constant() {r=2.5}
    B=MaterialFunctionCall!(Color=@n19, Shift=@n20) {materialFunction=/DealershipSharedLibrary/FX/Materials/MaterialFunctions/MF_MonochromaticShift}.Result
   )
   B=@n19
   Alpha=@MaterialExpressionConstantClamp_2
  )
  n22 = OneMinus(None=Multiply(A=StaticSwitchParameter(True=@n16, False=@n12) {parameterName=Trail?}, B=@n11))
  MaterialExpressionConstantClamp_3 = Saturate(None=ConstantBiasScale(None=Multiply(A=@n22, B=@MaterialExpressionConstantClamp_2)) {bias=-0.3, scale=4})
  MaterialExpressionConstantClamp_1 = Saturate(None=Add(A=@MaterialExpressionConstantClamp_3, B=@n9))
  MaterialFunctionCall(Emissive=LinearInterpolate(A=@n19, B=Multiply(A=@n19, B=@n21), Alpha=@MaterialExpressionConstantClamp_1)) {materialFunction=&MF_FX_EmissiveAdjustments}
  Multiply(A=@n3.G, B=@n6.R)
 ]
}
